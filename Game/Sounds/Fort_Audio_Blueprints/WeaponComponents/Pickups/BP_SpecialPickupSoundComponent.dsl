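; Sound component for special pickups that plays a tagged ambient loop and a pickup sound, on cosmetic-capable instances.
; The loop starts once game state is ready and stops on pickup or EndPlay.
; On pickup, the Sound Library sound plays only for a locally viewed pawn.
; Music-tagged sounds register with the music event subsystem, except the loop's component, which is stored on play and has its event removed on stop.
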(event EventBeginPlay
  (bind _asfort_pickup (Utilities|Casting|CastToFortPickup (Components|GetOwner)))
  (Interaction|BindEventtoOnPickup _asfort_pickup (EventDispatchers|CreateEvent :!event Custom|PickUpCustomAudio))
  (Class|FortAsyncActionGameStateReady|GameStateReadyAsync
    (:OnReady
      (Networking|CanExecuteCosmeticEvents
        (:True
          (CallFunction|PlayandRegisterAmbientLoop))))))

(event Custom|PickUpCustomAudio
  (params (SelfActor :type Fort%20Pickup%20Object%20Reference) (InteractingPawn :type Fort%20Pawn%20Object%20Reference))
  (Networking|CanExecuteCosmeticEvents
    (:True
      (CallFunction|StopAmbientLoop)
      (Utilities|IsValid InteractingPawn
        (:Is%20Valid
          (if (Pawn|IsLocallyViewed InteractingPawn)
            (bind _returnvalue (SoundLibrary|PlaySoundSimple (WorldSubsystems|GetSoundLibrarySubsystem) (Components|GetOwner) (Utilities|Struct|MakeSimpleSoundSettings "(TagName=\"SoundLibrary.Item.Pickup.Special\")" true false true false 1.0 1.0 0.0 0 0 "None" (Variables|/Script/Engine.Character|GetMesh InteractingPawn)) :!class /Script/SoundLibrary.SoundLibrarySubsystem))
            (if (and (.bSuccess (Utilities|Struct|BreakSoundLibraryPlayResult _returnvalue)) (Utilities|Array|IsNotEmpty (.AudioComponents (Utilities|Struct|BreakSoundLibraryPlayResult _returnvalue))))
              (if (AudioAssetUserDataTags|HasTag (Variables|/Script/Engine.AudioComponent|GetSound (Utilities|Array|Get(acopy) (.AudioComponents (Utilities|Struct|BreakSoundLibraryPlayResult _returnvalue)))) "(TagName=\"Audio.Metadata.Music\")")
                (Audio|AddEventbyComponent (GameInstanceSubsystems|GetMusicEventSubsystem) (Utilities|Array|Get(acopy) (.AudioComponents (Utilities|Struct|BreakSoundLibraryPlayResult _returnvalue))) "(TagName=\"MusicEventSubsystem.Event.Gameplay.Ability.1P\")" "(GameplayTags=)" 0 "")))))))))

(event EventEndPlay
  (params (EndPlayReason :type EEndPlayReason%20Enum :default "Destroyed"))
  (CallFunction|StopAmbientLoop))

(fn OnlyTriggerOnce)

(fn OnSoundPlayed
  (params (InEventName :type Gameplay%20Tag%20Structure) (InComponent :type Audio%20Component%20Object%20Reference))
  (if (GameplayTags|MatchesTag InEventName (Variables|Default|GetAmbientLoopEventTag) true)
    (Variables|Default|SetAmbientLoopAudioComponent InComponent)
    (if (AudioAssetUserDataTags|HasTag (Variables|/Script/Engine.AudioComponent|GetSound InComponent) "(TagName=\"Audio.Metadata.Music\")")
      (Audio|AddEventbyComponent (GameInstanceSubsystems|GetMusicEventSubsystem) InComponent "(TagName=\"MusicEventSubsystem.Event.POI.Diegetic\")" "(GameplayTags=)" 0 ""))))

(fn OnSoundStopped
  (params (InEventName :type Gameplay%20Tag%20Structure) (InComponent :type Audio%20Component%20Object%20Reference))
  (if (GameplayTags|MatchesTag InEventName (Variables|Default|GetAmbientLoopEventTag) true)
    (Audio|RemoveEvent (GameInstanceSubsystems|GetMusicEventSubsystem) InComponent)
    (return false)))

(fn PlayAndRegisterAmbientLoop
  (if (not (SoundLibrary|IsEventActive (Variables|Default|GetAmbientLoopEventTag)))
    (SoundLibrary|PlayEvent (Variables|Default|GetAmbientLoopEventTag))))

(fn StopAmbientLoop
  (SoundLibrary|StopEvent (Variables|Default|GetAmbientLoopEventTag)))
