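; Pickup sound component Blueprint that limits a special pickup's custom audio to once per pawn, for as long as a marker tag remains on it.
; On the custom pickup audio event, it checks whether the interacting pawn has the configurable TagToCheck gameplay tag.
; If the tag is absent, it calls the parent implementation, then adds that tag to the pawn as a loose gameplay tag to block repeats.
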(event EventPickUpCustomAudio
  (params (SelfActor :type Fort%20Pickup%20Object%20Reference) (InteractingPawn :type Fort%20Pawn%20Object%20Reference))
  (if (not (GameplayTags|HasMatchingGameplayTag InteractingPawn (Variables|Default|GetTagToCheck)))
    (|CallParentFunction SelfActor InteractingPawn)
    (Ability|GameplayEffect|AddLooseGameplayTags InteractingPawn (GameplayTags|MakeGameplayTagContainerfromTag (Variables|Default|GetTagToCheck)))))
